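# Niagara burst effect for the Veiled backpack cosmetic: a glitchy, pixelated reactive flash placed on the backpack's skeletal mesh.
# A sprite glow and a time-of-day-scaled mesh glow with a light accompany grid-snapped digital squares and rising pixel-art shapes, both shrinking in stepped curves.
# Glitch_sprite has zero spawn rate, and three other emitters are disabled, including the only users of the ColorArray and EmissiveColorSelection parameters.
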NiagaraSystem /BRCosmetics/Accessories/FORT_Backpacks/Backpack_F_MED_Veiled/FX/NS_Veiled_Reactive_Burst
dependencies {
 usedDataInterfaces [
  &NiagaraDataInterfaceCurve
  &NiagaraDataInterfaceArrayFloat
  &NiagaraDataInterfaceColorCurve
  &NiagaraDataInterfaceVector2DCurve
 ]
 usedDynamicInputs [
  &MakeCustomVectorFromBool
  &RandomBool
  &Vector2DFromCurve
 ]
 usedModules [&SystemState]
}
emitters {
 Cubes {
  bEnabled=false
  emitterUpdateScript [
   &EmitterState(Life Cycle Mode=::NewEnumerator1, Loop Behavior=::NewEnumerator1, Loop Delay=0.08, Loop Duration=0.2, UseLoopDelay=true)
   &SpawnRate(
    SpawnRate=&FloatFromCurve(
     FloatCurve=DataInterface() {propertyValues {Curve {keys [<at 0.05 0.5>,<at 0.2 1>,<at 1 0>]}}}
     CurveIndex=$Emitter.NormalizedLoopAge
     Scale Curve=100
    )
    Use Spawn Probability=false
   )
   &SpawnBurst_Instantaneous(
    Spawn Count=&RandomRangeInt(Minimum=12, Maximum=14)
    Use Spawn Probability=false
    UseLoopCountLimit=false
   ) {enabled=false}
  ]
  particleSpawnScript [
   &InitializeParticle(
    Color=&SelectLinearColorFromArray(Array Sampling Mode=::NewEnumerator0, Direct Array Index=?, Color Selection Array=$User.ColorArray, Random Seed=0)
    Lifetime=&Multiply_Float(
     A=&FloatFromCurve(FloatCurve=DataInterface() {propertyValues {Curve {keys [<at 0 1>,<at 0.5 0.3>]}}}, CurveIndex=$Emitter.NormalizedLoopAge)
     B=&RandomRangeFloat(Minimum=0.2, Maximum=0.5)
    )
    Sprite Rotation Angle=&RandomRangeFloat(Minimum=0, Maximum=0)
    Sprite Rotation Mode=::NewEnumerator1
    Sprite Size Mode=::NewEnumerator3
    Uniform Sprite Size=&FloatFromCurve(
     FloatCurve=DataInterface() {propertyValues {Curve {keys [<at 0 1.5>,<at 0.2 1.5>,<at 0.8 4>,<at 1 4>]}}}
     CurveIndex=&RandomRangeFloat(Minimum=0, Maximum=1)
    )
    UsePositionOffset=false
    Write Material Random=true
   )
   &InheritVelocity(Inherited Velocity Amount Scale=&VectorFromFloat(Value=0.5), Speed Limit=true, UseSpeedThreshold=false)
   &ShapeLocation(
    MaskSpawnGroup=false
    Override Seed=false
    Sphere Radius=&RandomRangeFloat(Minimum=1, Maximum=4)
    UseOwnerScale=true
   ) {enabled=false}
   &AddVelocityFromPoint(Module.Use Falloff Distance=false, Velocity Strength=&RandomRangeFloat(Minimum=400, Maximum=500)) {enabled=false}
   &ApplyInitialForces()
   &CameraOffset(Camera Offset Amount=3)
   &SkeletalMeshLocation(
    MaskSpawnGroup=false
    Mesh Sampling Type=::NewEnumerator5
    Sampled Position Offset=&RandomRangeVector(Minimum=<0 0 0>, Maximum=<0 0 0>)
    Sampled Position Offset Coordinate Space:ENiagaraSkelMeshTransforms=::NewEnumerator2
    Skeletal Mesh=$User.SkeletalMesh
    Socket Sampling Mode=::NewEnumerator5
   )
   &ShapeLocation(MaskSpawnGroup=false, Override Seed=false, Sphere Radius=10, UseOwnerScale=true) {module=ShapeLocation001}
  ]
  particleUpdateScript [
   &Drag(Drag=&RandomRangeFloat(Minimum=10, Maximum=11), Use Linear Drag=true, Use Rotational Drag=false)
   &SolveForcesAndVelocity(Clamp Velocity=false, Limit Acceleration=false)
   &ParticleState()
   &ScaleSpriteSize(
    Uniform Curve Scale=2
    Uniform Curve Sprite Scale=DataInterface() {propertyValues {Curve {keys [
        {arriveTangent=5.201, interpMode=RCIM_Constant, time=0, value=1}
        <at 0.4 0.5 RCIM_Constant>
        <at 0.75 0.25 RCIM_Constant>
        {arriveTangent=-1.983, interpMode=RCIM_Constant, leaveTangent=-1.983, time=1, value=0}
       ]}}}
   )
   &DynamicMaterialParameters(
    Index 0 Param 1=0.02
    Module.Param0WriteEnabled=true
    Module.Param1WriteEnabled=true
    Module.Param2WriteEnabled=true
    Module.Param3WriteEnabled=true
   )
   &self:Quantizer_06(SnapDistance=4) {module=Quantizer_06}
   /SetVariables(Particles.Lifetime=&Multiply_Float(
     A=$Particles.Initial.Lifetime
     B=&FloatFromCurve(FloatCurve=DataInterface() {propertyValues {Curve {keys [<at 0 1>,<at 1 0.3>]}}}, CurveIndex=$Emitter.NormalizedLoopAge)
    )) {enabled=false, isSetParameters=true, module=SetVariables_#1}
   &ScaleColor(
    Scale Alpha=&FloatFromCurve(FloatCurve=DataInterface() {propertyValues {Curve {keys [<at 0 1>,<at 1 0>]}}})
    Scale RGB=&VectorFromFloat(Value=&FloatFromCurve(FloatCurve=DataInterface() {propertyValues {Curve {keys [<at 0 1>,<at 1 0>]}}}, Scale Curve=0.3))
    ScaleA=true
    ScaleRGB=true
   )
  ]
  rendererClasses [&NiagaraSpriteRendererProperties]
  renderers [{rendererClass=&NiagaraSpriteRendererProperties}]
 }
 DigitalSquares {
  bEnabled=false
  emitterUpdateScript [
   &EmitterState(Life Cycle Mode=::NewEnumerator1, Loop Behavior=::NewEnumerator1, Loop Duration=0.3, UseLoopDelay=false)
   &SpawnBurst_Instantaneous(Spawn Count=10, Use Spawn Probability=false, UseLoopCountLimit=false)
  ]
  particleSpawnScript [
   &InitializeParticle(
    Color Mode=::NewEnumerator0
    Lifetime=&RandomRangeFloat(Minimum=0.3, Maximum=1.1)
    Sprite Rotation Angle=45
    Sprite Rotation Mode=::NewEnumerator0
    Sprite Size Mode=::NewEnumerator3
    Uniform Sprite Size=&SelectFloatFromArray(
     Array Sampling Mode=::NewEnumerator0
     Direct Array Index=?
     Float Selection Array=DataInterface() {propertyValues {FloatData [3,5,7]}}
     Random Seed=0
    )
    UsePositionOffset=false
    Write Material Random=true
   )
   /SetVariables(Particles.SubImageIndex=3) {isSetParameters=true}
   &ShapeLocation(
    MaskSpawnGroup=false
    Offset Mode=::NewEnumerator0
    Override Seed=false
    Ring Radius=1
    Rotation Axis=<0 -1 0>
    Rotation Mode=::NewEnumerator1
    Shape Primitive=::NewEnumerator4
    UseOwnerScale=true
   ) {enabled=false}
   &ShapeLocation(
    MaskSpawnGroup=false
    Non Uniform Scale=<1 1 0.2>
    Offset Mode=::NewEnumerator0
    Rotation Axis=<0 -1 0>
    Rotation Mode=::NewEnumerator1
    Sphere Distribution=::NewEnumerator3
    Sphere Radius=0.5
    Uniform Spiral Amount:NiagaraFloat=&RandomRangeFloat(Minimum=1, Maximum=3)
    UseOwnerScale=true
   ) {module=ShapeLocation001}
   &SphereLocation(
    MaskSpawnGroup=false
    Offset=<0 -2 0>
    Sphere Distribution:ENiagaraSphereDistributionMode=::NewEnumerator3
    Sphere Radius=3
    UseNonUniformScale=false
    UseOffset=true
    UseSphereOrientation=false
   ) {enabled=false}
   &AddVelocityFromPoint(Module.Use Falloff Distance=false)
  ]
  particleUpdateScript [
   &ParticleState()
   &ScaleSpriteSize(
    Uniform Curve Scale=0.8
    Uniform Curve Sprite Scale=DataInterface() {propertyValues {Curve {keys [
        {interpMode=RCIM_Cubic, tangentMode=RCTM_User, time=0, value=0}
        {interpMode=RCIM_Cubic, tangentMode=RCTM_User, time=0.15, value=3}
        <at 0.3 1.1>
        {interpMode=RCIM_Cubic, tangentMode=RCTM_User, time=1, value=2}
       ]}}}
   ) {module=ScaleSpriteSize001}
   &Drag(Drag=2, Use Linear Drag=true, Use Rotational Drag=false)
   /SetVariables(
    Particles.ParticleColorCurve1:LinearColor=&ColorFromCurve(ColorCurve=DataInterface() {propertyValues {
       AlphaCurve {keys [<at 0.3132 1 RCIM_Cubic>,<at 1 0 RCIM_Cubic>]}
       BlueCurve {keys [<at 0.2036 0.2667 RCIM_Cubic>,<at 0.3445 2>,<at 0.5237 2 RCIM_Cubic>]}
       GreenCurve {keys [{arriveTangent=-14.1913, interpMode=RCIM_Cubic, time=0.2036, value=2}, <at 0.3445 0>, <at 0.5237 0 RCIM_Cubic>]}
       RedCurve {keys [<at 0.2036 0 RCIM_Cubic>, <at 0.3445 0>, {interpMode=RCIM_Cubic, leaveTangent=11.1602, time=0.5237, value=2}]}
      }})
    Particles.ParticleColorCurve2:LinearColor=&ColorFromCurve(ColorCurve=DataInterface() {propertyValues {
       AlphaCurve {keys [<at 0.3132 1 RCIM_Cubic>,<at 1 0 RCIM_Cubic>]}
       BlueCurve {keys [
         {arriveTangent=5.44, interpMode=RCIM_Cubic, leaveTangent=5.44, time=0.2036, value=0.5}
         <at 0.3445 1.2667>
         {arriveTangent=4.0921, interpMode=RCIM_Cubic, leaveTangent=4.0921, time=0.5237, value=2}
        ]}
       GreenCurve {keys [<at 0.2036 0 RCIM_Cubic>,<at 0.3445 0>,<at 0.5237 0 RCIM_Cubic>]}
       RedCurve {keys [<at 0.2036 2 RCIM_Cubic>,<at 0.3445 2>,<at 0.5237 2 RCIM_Cubic>]}
      }})
    Particles.ParticleColorCurve3:LinearColor=&ColorFromCurve(ColorCurve=DataInterface() {propertyValues {
       AlphaCurve {keys [<at 0.3132 1 RCIM_Cubic>,<at 1 0 RCIM_Cubic>]}
       BlueCurve {keys [
         {arriveTangent=0.2365, interpMode=RCIM_Cubic, leaveTangent=0.2365, time=0.2036, value=0}
         <at 0.3445 0.0333>
         {arriveTangent=10.9742, interpMode=RCIM_Cubic, leaveTangent=10.9742, time=0.5237, value=2}
        ]}
       GreenCurve {keys [{arriveTangent=-8.9879, interpMode=RCIM_Cubic, leaveTangent=-8.9879, time=0.2036, value=1.2667}, <at 0.3445 0>, <at 0.5237 0 RCIM_Cubic>]}
       RedCurve {keys [<at 0.2036 2 RCIM_Cubic>,<at 0.3445 2>,<at 0.5237 2 RCIM_Cubic>]}
      }})
    Particles.ParticleColorCurve4:LinearColor=&ColorFromCurve(ColorCurve=DataInterface() {propertyValues {
       AlphaCurve {keys [<at 0.3132 1 RCIM_Cubic>,<at 1 0 RCIM_Cubic>]}
       BlueCurve {keys [<at 0.2036 2 RCIM_Cubic>,<at 0.3445 2>,<at 0.5237 2 RCIM_Cubic>]}
       GreenCurve {keys [
         {arriveTangent=-10.6435, interpMode=RCIM_Cubic, leaveTangent=-10.6435, time=0.2036, value=2}
         <at 0.3445 0.5>
         {arriveTangent=-2.7901, interpMode=RCIM_Cubic, leaveTangent=-2.7901, time=0.5237, value=0}
        ]}
       RedCurve {keys [<at 0.2036 2 RCIM_Cubic>,<at 0.3445 2>,<at 0.5237 2 RCIM_Cubic>]}
      }})
   ) {isSetParameters=true}
   &Color(Color=&LerpLinearColors(
     StartColor=&LerpLinearColors(
      StartColor=$Particles.ParticleColorCurve1
      EndColor=$Particles.ParticleColorCurve2
      LerpFactor=/Niagara/DynamicInputs/Integer/V2/Round_Float(Float:NiagaraFloat=&ClampFloat(Float=&Subtract_Float(A=$User.EmissiveColorSelection, B=0)))
     )
     EndColor=&LerpLinearColors(
      StartColor=$Particles.ParticleColorCurve3
      EndColor=$Particles.ParticleColorCurve4
      LerpFactor=&Round_Float(Float:NiagaraFloat=&ClampFloat(Float=&Subtract_Float(A=$User.EmissiveColorSelection, B=2)))
     )
     LerpFactor=&Round_Float(Float:NiagaraFloat=&ClampFloat(Float=&Subtract_Float(A=$User.EmissiveColorSelection, B=1)))
    )) {module=Color001}
   &SolveForcesAndVelocity(Clamp Velocity=false, Limit Acceleration=false)
   &self:Quantizer(SnapDistance=10) {module=Quantizer}
  ]
  rendererClasses [&NiagaraSpriteRendererProperties]
  renderers [{rendererClass=&NiagaraSpriteRendererProperties}]
 }
 DigitalSquares003 {
  emitterUpdateScript [
   &EmitterState(Life Cycle Mode=::NewEnumerator1, Loop Behavior=::NewEnumerator1, Loop Duration=0.4, UseLoopDelay=false) {module=EmitterState001}
   &EmitterState(Life Cycle Mode=::NewEnumerator1, Loop Behavior=::NewEnumerator1, Loop Duration=0.3, UseLoopDelay=false) {enabled=false}
   &SpawnRate(SpawnRate=20, Use Spawn Probability=false)
   &SpawnBurst_Instantaneous(Spawn Count=2, Use Spawn Probability=false, UseLoopCountLimit=false)
  ]
  particleSpawnScript [
   &InitializeParticle(
    Color Mode=::NewEnumerator0
    Lifetime=&RandomRangeFloat(Minimum=0.3, Maximum=0.5)
    Sprite Rotation Mode=::NewEnumerator0
    Sprite Size Mode=::NewEnumerator3
    Uniform Sprite Size=&SelectFloatFromArray(
     Array Sampling Mode=::NewEnumerator0
     Direct Array Index=?
     Float Selection Array=DataInterface() {propertyValues {FloatData [3,5,8]}}
     Random Seed=0
    )
    UsePositionOffset=false
    Write Material Random=true
   )
   &ShapeLocation(
    MaskSpawnGroup=false
    Offset Mode=::NewEnumerator0
    Override Seed=false
    Ring Radius=10
    Rotation Axis=<0 -1 0>
    Rotation Mode=::NewEnumerator1
    Shape Primitive=::NewEnumerator4
    UseOwnerScale=true
   ) {enabled=false}
   &SkeletalMeshLocation(
    Apply Sampled Orientation to Mesh Particles=true
    Bone Sampling Mode=::NewEnumerator4
    Inherit Velocity=false
    "Inherit Velocity (Scale)"=0
    Interpolate Mesh Sampling=false
    MaskSpawnGroup=false
    Sampled Position Offset Coordinate Space:ENiagaraMeshTransforms=::NewEnumerator2
    Skeletal Mesh=$User.SkeletalMesh
   ) {module=SkeletalMeshLocation001}
   &ShapeLocation(
    Hemisphere Distribution=&MakeVector2D(
     X=0.5
     Y=&Lerp_Float(A=1, B=&Multiply_Float(A=0.5, B=&ReturnNormalizedExecIndex()), Alpha=0)
    )
    MaskSpawnGroup=false
    Non Uniform Scale=<1 -1 1>
    Override Seed=false
    Sphere Radius=15
    Sphere Surface Distribution=1
    UseOwnerScale=true
   ) {module=ShapeLocation003}
   &ShapeLocation(
    MaskSpawnGroup=false
    Override Seed=false
    Ring Radius=&RandomRangeFloat(Minimum=10, Maximum=15)
    Rotation Axis=<0 1 0>
    Rotation Mode=::NewEnumerator1
    Shape Primitive=::NewEnumerator4
    UseOwnerScale=true
   ) {enabled=false, module=ShapeLocation005}
   &ShapeLocation(MaskSpawnGroup=false, Non Uniform Scale=<1 -1 1>, Sphere Distribution=::NewEnumerator3, Sphere Radius=10, UseOwnerScale=true) {enabled=false, module=ShapeLocation004}
   &ShapeLocation(
    Hemisphere Distribution=<0 0>
    MaskSpawnGroup=false
    Offset Mode=::NewEnumerator0
    Override Seed=false
    Rotation Axis=<0 -1 0>
    Rotation Mode=::NewEnumerator1
    Sphere Radius=1
    Sphere Surface Distribution=1
    UseOwnerScale=true
   ) {enabled=false, module=ShapeLocation001}
   &ShapeLocation(MaskSpawnGroup=false, Override Seed=false, UseOwnerScale=true) {enabled=false, module=ShapeLocation002}
   &SphereLocation(
    MaskSpawnGroup=false
    Offset=<0 -2 0>
    Override Seed=false
    Sphere Origin:NiagaraPosition=$Output.SkeletalMeshLocation001.SampledPosition
    Sphere Radius=0
    Surface Only=false
    UseNonUniformScale=false
    UseOffset=true
    UseSphereOrientation=false
   ) {enabled=false}
   &AddVelocityFromPoint(Module.Use Falloff Distance=false, Velocity Strength=&RandomRangeFloat(Minimum=0, Maximum=0))
   &SubUVAnimation(End Frame=10, SubUV Animation Mode:ENiagara_SubUVLookupMode=::NewEnumerator2)
   &CameraOffset(Camera Offset Amount=1) {enabled=false}
   &DynamicMaterialParameters(
    Index 0 Param 1=&RandomRangeFloat(Minimum=0, Maximum=1)
    Module.Param0WriteEnabled=true
    Module.Param1WriteEnabled=true
    Module.Param2WriteEnabled=true
    Module.Param3WriteEnabled=true
   )
  ]
  particleUpdateScript [
   &ParticleState()
   &SubUVAnimation(End Frame=4) {module=SubUVAnimation002}
   &SubUVAnimation(
    End Frame:NiagaraInt32=&RandomRangeInt(Minimum=0, Maximum=30)
    Start Frame:NiagaraInt32=&RandomRangeInt(Minimum=0, Maximum=10)
   ) {enabled=false, module=SubUVAnimation001}
   &ScaleSpriteSize(
    Uniform Curve Scale=0.8
    Uniform Curve Sprite Scale=DataInterface() {propertyValues {Curve {keys [<at 0 1.5 RCIM_Constant>,<at 0.25 1 RCIM_Constant>,<at 0.5 0.4 RCIM_Constant>,<at 0.75 0.25 RCIM_Constant>,<at 1 0 RCIM_Constant>]}}}
   ) {enabled=false, module=ScaleSpriteSize001}
   &ScaleSpriteSize(
    Uniform Curve Scale=0.8
    Uniform Curve Sprite Scale=DataInterface() {propertyValues {Curve {keys [<at 0 1.5 RCIM_Constant>,<at 0.5 0.4 RCIM_Constant>,<at 1 0 RCIM_Constant>]}}}
   ) {module=ScaleSpriteSize002}
   &ScaleSpriteSize(
    Uniform Curve Scale=0.8
    Uniform Curve Sprite Scale=DataInterface() {propertyValues {Curve {keys [
        {tangentMode=RCTM_User, time=0, value=1}
        {arriveTangent=-0.0405, interpMode=RCIM_Cubic, leaveTangent=-0.0405, tangentMode=RCTM_User, time=0.7009, value=1}
        {arriveTangent=-8.0089, interpMode=RCIM_Cubic, leaveTangent=-8.0089, tangentMode=RCTM_User, time=1, value=0}
       ]}}}
   )
   &Drag(Drag=2, Use Linear Drag=true, Use Rotational Drag=false)
   &SolveForcesAndVelocity(Clamp Velocity=false, Limit Acceleration=false)
   &self:Quantizer_04(SnapDistance=4) {module=Quantizer_04}
   &ScaleColor(
    Scale RGB=&VectorFromFloat(Value=&FloatFromCurve(FloatCurve=DataInterface() {propertyValues {Curve {keys [
          <at 0 1 RCIM_Cubic>
          {arriveTangent=-0.8, interpMode=RCIM_Cubic, leaveTangent=-0.8, time=0.2, value=1}
          {arriveTangent=-2.4671, interpMode=RCIM_Cubic, leaveTangent=-2.4671, tangentMode=RCTM_User, time=1, value=0.2}
         ]}}}))
    ScaleA=true
    ScaleRGB=true
   )
  ]
  rendererClasses [&NiagaraSpriteRendererProperties]
  renderers [{rendererClass=&NiagaraSpriteRendererProperties}]
 }
 Glitch_sprite {
  emitterUpdateScript [
   &EmitterState(Life Cycle Mode=::NewEnumerator1, Loop Behavior=::NewEnumerator1, Loop Delay=0.05, Loop Duration=0.2, UseLoopDelay=true)
   &SpawnBurst_Instantaneous(Spawn Time=0.175, Use Spawn Probability=false, UseLoopCountLimit=false) {enabled=false}
   &SpawnRate(
    SpawnRate=&Multiply_Float(A=0, B=&FN_FrontEndFloatModifier(Gameplay=1, Front End=2))
    Use Spawn Probability=false
   )
  ]
  particleSpawnScript [
   &InitializeParticle(
    Color=<rgba 5 0.8096 0.8795 2>
    Lifetime Max=0.3
    Lifetime Min=0.3
    Lifetime Mode=::NewEnumerator1
    Mesh Scale Mode=::NewEnumerator1
    Position Offset=<0 0 100>
    UsePositionOffset=true
    Write Material Random=true
   )
   &SkeletalMeshLocation(
    Bone Sampling Mode=::NewEnumerator4
    Inherit Velocity=false
    "Inherit Velocity (Scale)"=0
    Interpolate Mesh Sampling=false
    MaskSpawnGroup=false
    Sampled Position Offset Coordinate Space:ENiagaraMeshTransforms=::NewEnumerator2
    Skeletal Mesh=$User.SkeletalMesh
   ) {module=SkeletalMeshLocation001}
   &CameraOffset(Camera Offset Amount=50)
   &ShapeLocation(
    Box Size=&Multiply_Vector(
     A=&VectorFromFloat(Value=0.25)
     B=&FN_FrontEndVector3Modifier(Gameplay=<62 62 100>, Front End=<80 80 100>)
    )
    MaskSpawnGroup=false
    Override Seed=false
    Shape Primitive=::NewEnumerator2
    Surface Only=false
    UseOwnerScale=true
   ) {enabled=false}
   &ShapeLocation(
    Hemisphere Distribution=&MakeVector2D(
     X=0.5
     Y=&Lerp_Float(A=1, B=&Multiply_Float(A=0.5, B=&ReturnNormalizedExecIndex()), Alpha=0)
    )
    MaskSpawnGroup=false
    Non Uniform Scale=<1 -1 1>
    Override Seed=false
    Sphere Radius=10
    Sphere Surface Distribution=1
    UseOwnerScale=true
   ) {module=ShapeLocation003}
  ]
  particleUpdateScript [
   &ParticleState()
   &ScaleMeshSize(Scale Factor=&VectorFromFloat(Value=&FloatFromCurve(
      FloatCurve=DataInterface() {propertyValues {Curve {keys [<at 0 1.5 RCIM_Constant>,<at 0.25 0.75 RCIM_Constant>,<at 0.5 0.5 RCIM_Constant>,<at 0.75 0.25 RCIM_Constant>,<at 1 0 RCIM_Constant>]}}}
      Scale Curve=0.1
     ))) {enabled=false, module=ScaleMeshSize001}
   &ScaleMeshSize(Scale Factor=&Multiply_VectorByFloat(
     Vector=<1 1 1>
     Float=&Multiply_Float(
      A=0.2
      B=&FloatFromCurve(
       FloatCurve=DataInterface() {propertyValues {Curve {keys [
           {tangentMode=RCTM_User, time=0, value=1}
           {arriveTangent=-0.0405, interpMode=RCIM_Cubic, leaveTangent=-0.0405, tangentMode=RCTM_User, time=0.7009, value=1}
           {arriveTangent=-8.0089, interpMode=RCIM_Cubic, leaveTangent=-8.0089, tangentMode=RCTM_User, time=1, value=0}
          ]}}}
       Scale Curve=0.5
      )
     )
    ))
   &Drag(Drag=2, Use Linear Drag=true, Use Rotational Drag=true)
   &SolveForcesAndVelocity(Clamp Velocity=false, Limit Acceleration=false)
   &DynamicMaterialParameters(Module.Param0WriteEnabled=true, Module.Param1WriteEnabled=true, Module.Param2WriteEnabled=true, Module.Param3WriteEnabled=true)
   &ScaleColor(
    Scale Alpha=&Multiply_Float(
     A=&FloatFromCurve(
      FloatCurve=DataInterface() {propertyValues {Curve {keys [
          {arriveTangent=9.0725, interpMode=RCIM_Cubic, leaveTangent=9.0725, tangentMode=RCTM_User, time=0, value=0}
          {arriveTangent=-2.5249, interpMode=RCIM_Cubic, leaveTangent=-2.5249, tangentMode=RCTM_User, time=0.2, value=0.9847}
          {arriveTangent=-0.0405, interpMode=RCIM_Cubic, leaveTangent=-0.0405, tangentMode=RCTM_User, time=1, value=0}
         ]}}}
      Scale Curve=&RandomRangeFloat(Minimum=0.2, Maximum=1)
     )
     B=0.5
    )
    Scale RGB=&Multiply_VectorByFloat(
     Vector=&MakeVectorFromLinearColorRGB(Color=&ColorFromCurve(ColorCurve=DataInterface() {propertyValues {
         AlphaCurve {keys []}
         BlueCurve {keys [
           <at 0 5 RCIM_Cubic>
           <at 0.1326 5>
           <at 0.2869 5>
           <at 0.4134 0>
           <at 0.5108 5>
           <at 0.6236 0.0803>
           {arriveTangent=26.436, interpMode=RCIM_Cubic, leaveTangent=26.436, time=0.7107, value=0.8321}
           <at 0.8097 5>
          ]}
         GreenCurve {keys [
           <at 0 4.702 RCIM_Cubic>
           <at 0.1326 0>
           <at 0.2869 0.6611>
           <at 0.4134 0.0442>
           <at 0.5108 0.7585>
           <at 0.6236 5>
           {arriveTangent=-21.532, interpMode=RCIM_Cubic, leaveTangent=-21.532, time=0.7107, value=1}
           <at 0.8097 0.9929>
          ]}
         RedCurve {keys [<at 0 0 RCIM_Cubic>,<at 0.1326 2.465>,<at 0.2869 0>,<at 0.4134 5>,<at 0.5108 4.6237>,<at 0.6236 0>,<at 0.7107 0 RCIM_Cubic>,<at 0.8097 0>]}
        }}))
     Float=&FloatFromCurve(
      FloatCurve=DataInterface() {propertyValues {Curve {keys [{arriveTangent=-2.0684, interpMode=RCIM_Cubic, leaveTangent=-2.0684, tangentMode=RCTM_User, time=0, value=2}, <at 1 0>]}}}
      Scale Curve=&Multiply_Float(A=1, B=&FN_FrontEndFloatModifier(Gameplay=1, Front End=1))
     )
    )
    ScaleA=true
    ScaleRGB=true
   )
   &ScaleSpriteSize(Uniform Curve Sprite Scale=DataInterface() {propertyValues {Curve {keys [<at 0 1.5 RCIM_Constant>,<at 0.25 1 RCIM_Constant>,<at 0.5 0.4 RCIM_Constant>,<at 0.75 0.25 RCIM_Constant>,<at 1 0 RCIM_Constant>]}}}) {enabled=false, module=ScaleSpriteSize001}
   &ScaleSpriteSize(
    Uniform Curve Scale=0.8
    Uniform Curve Sprite Scale=DataInterface() {propertyValues {Curve {keys [<at 0 1.5 RCIM_Constant>,<at 0.5 0.4 RCIM_Constant>,<at 1 0 RCIM_Constant>]}}}
   ) {enabled=false}
  ]
  rendererClasses [&NiagaraMeshRendererProperties]
  renderers [{rendererClass=&NiagaraMeshRendererProperties, rendererIndex=0}]
 }
 Glitch_sprite001 {
  bEnabled=false
  emitterUpdateScript [
   &EmitterState(Life Cycle Mode=::NewEnumerator1, Loop Behavior=::NewEnumerator1, Loop Delay=0.05, Loop Duration=0.4, UseLoopDelay=true)
   &SpawnRate(
    SpawnRate=&Multiply_Float(A=10, B=&FN_FrontEndFloatModifier(Gameplay=1, Front End=2))
    Use Spawn Probability=false
   )
  ]
  particleSpawnScript [
   &InitializeParticle(
    Color=<rgba 5 0.8096 0.8795 2>
    Lifetime Max=0.3
    Lifetime Min=0.3
    Lifetime Mode=::NewEnumerator1
    Mesh Scale Mode=::NewEnumerator1
    Mesh Uniform Scale Max=4
    Mesh Uniform Scale Min=2
    Position Offset=<0 0 100>
    UsePositionOffset=true
    Write Material Random=true
   )
   &SkeletalMeshLocation(
    Bone Sampling Mode=::NewEnumerator4
    Inherit Velocity=false
    "Inherit Velocity (Scale)"=0
    Interpolate Mesh Sampling=false
    MaskSpawnGroup=false
    Sampled Position Offset Coordinate Space:ENiagaraMeshTransforms=::NewEnumerator2
    Skeletal Mesh=$User.SkeletalMesh
   ) {module=SkeletalMeshLocation001}
   &CameraOffset(Camera Offset Amount=50)
   &ShapeLocation(
    Box Size=&Multiply_Vector(
     A=&VectorFromFloat(Value=0.25)
     B=&FN_FrontEndVector3Modifier(Gameplay=<62 62 100>, Front End=<80 80 100>)
    )
    MaskSpawnGroup=false
    Override Seed=false
    Shape Primitive=::NewEnumerator2
    Surface Only=false
    UseOwnerScale=true
   ) {enabled=false}
   &ShapeLocation(
    Hemisphere Distribution=&MakeVector2D(
     X=0.5
     Y=&Lerp_Float(A=1, B=&Multiply_Float(A=0.5, B=&ReturnNormalizedExecIndex()), Alpha=0)
    )
    MaskSpawnGroup=false
    Non Uniform Scale=<1 -1 1>
    Override Seed=false
    Sphere Radius=15
    Sphere Surface Distribution=1
    UseOwnerScale=true
   ) {module=ShapeLocation003}
  ]
  particleUpdateScript [
   &ParticleState()
   &ScaleMeshSize(Scale Factor=&VectorFromFloat(Value=&FloatFromCurve(FloatCurve=DataInterface() {propertyValues {Curve {keys [<at 0 1>,<at 1 0>]}}}, Scale Curve=0.1))) {module=ScaleMeshSize001}
   &ScaleMeshSize(Scale Factor=&Multiply_VectorByFloat(
     Vector=<1 1 1>
     Float=&Multiply_Float(
      A=0.2
      B=&FloatFromCurve(
       FloatCurve=DataInterface() {propertyValues {Curve {keys [
           {arriveTangent=4.1785, time=0, value=1}
           <at 0.5 1.1>
           {arriveTangent=-0.0405, interpMode=RCIM_Cubic, leaveTangent=-0.0405, tangentMode=RCTM_User, time=1, value=1.5}
          ]}}}
       Scale Curve=0.5
      )
     )
    )) {enabled=false}
   &Drag(Drag=2, Use Linear Drag=true, Use Rotational Drag=true)
   &SolveForcesAndVelocity(Clamp Velocity=false, Limit Acceleration=false)
   &DynamicMaterialParameters(Module.Param0WriteEnabled=true, Module.Param1WriteEnabled=true, Module.Param2WriteEnabled=true, Module.Param3WriteEnabled=true)
   &ScaleColor(
    Scale Alpha=&Multiply_Float(
     A=&FloatFromCurve(
      FloatCurve=DataInterface() {propertyValues {Curve {keys [
          {arriveTangent=9.0725, interpMode=RCIM_Cubic, leaveTangent=9.0725, tangentMode=RCTM_User, time=0, value=0}
          {arriveTangent=-2.5249, interpMode=RCIM_Cubic, leaveTangent=-2.5249, tangentMode=RCTM_User, time=0.2, value=0.9847}
          {arriveTangent=-0.0405, interpMode=RCIM_Cubic, leaveTangent=-0.0405, tangentMode=RCTM_User, time=1, value=0}
         ]}}}
      Scale Curve=&RandomRangeFloat(Minimum=0.2, Maximum=1)
     )
     B=0.5
    )
    Scale RGB=&Multiply_VectorByFloat(
     Vector=&MakeVectorFromLinearColorRGB(Color=&ColorFromCurve(ColorCurve=DataInterface() {propertyValues {
         AlphaCurve {keys []}
         BlueCurve {keys [
           <at 0 5 RCIM_Cubic>
           <at 0.1326 5>
           <at 0.2869 5>
           <at 0.4134 0>
           <at 0.5108 5>
           <at 0.6236 0.0803>
           {arriveTangent=26.436, interpMode=RCIM_Cubic, leaveTangent=26.436, time=0.7107, value=0.8321}
           <at 0.8097 5>
          ]}
         GreenCurve {keys [
           <at 0 4.702 RCIM_Cubic>
           <at 0.1326 0>
           <at 0.2869 0.6611>
           <at 0.4134 0.0442>
           <at 0.5108 0.7585>
           <at 0.6236 5>
           {arriveTangent=-21.532, interpMode=RCIM_Cubic, leaveTangent=-21.532, time=0.7107, value=1}
           <at 0.8097 0.9929>
          ]}
         RedCurve {keys [<at 0 0 RCIM_Cubic>,<at 0.1326 2.465>,<at 0.2869 0>,<at 0.4134 5>,<at 0.5108 4.6237>,<at 0.6236 0>,<at 0.7107 0 RCIM_Cubic>,<at 0.8097 0>]}
        }}))
     Float=&FloatFromCurve(
      FloatCurve=DataInterface() {propertyValues {Curve {keys [{arriveTangent=-2.0684, interpMode=RCIM_Cubic, leaveTangent=-2.0684, tangentMode=RCTM_User, time=0, value=2}, <at 1 0>]}}}
      Scale Curve=&Multiply_Float(A=1, B=&FN_FrontEndFloatModifier(Gameplay=1, Front End=1))
     )
    )
    ScaleA=true
    ScaleRGB=true
   )
  ]
  rendererClasses [&NiagaraMeshRendererProperties]
  renderers [{rendererClass=&NiagaraMeshRendererProperties, rendererIndex=0}]
 }
 Glow {
  emitterSpawnScript [/SetVariables(Emitter.Depthfade=10, Emitter.Hardness=0.15, Emitter.Power=1.25) {isSetParameters=true}]
  emitterUpdateScript [
   &EmitterState(Life Cycle Mode=::NewEnumerator1, Loop Behavior=::NewEnumerator1, Loop Duration=0, UseLoopDelay=false)
   &SpawnBurst_Instantaneous(Use Spawn Probability=false, UseLoopCountLimit=false)
   /SetVariables(Emitter.Power:NiagaraFloat=&Add_Float(
     A=1.25
     B=&Waveform(
      Waveform Count=::NewEnumerator1
      "Waveform [1]"=::NewEnumerator0
      "[1] Phase Scale"=0.2215
      "[1] Frequency"=0.451
      "[1] Amplitude Min/Max"=<0 1>
      "[1] Scale"=0.95
      "[2] Average"=false
      "[2] Input Position"=0
      "[2] Phase"=$Emitter.Age
      "[2] Phase Scale"=0.234
      "[2] Compound Frequency"=2
      "[2] Frequency"=0.521
      "[2] Amplitude Min/Max"=<0 1>
      "[2] Scale"=1
      "[2] Bias"=0
      "[2] Exponent"=1
      "[2] Pulse Duration"=0.1
      "[2] Clip Triangle Min/Max"=<0 0>
      "[2] Random Seed"=0
      "[3] Phase"=$Emitter.Age
     )
    )) {enabled=false, isSetParameters=true}
  ]
  particleSpawnScript [
   &InitializeParticle(
    Color=<rgba 1 0 0.7 0.2>
    Lifetime=0.5
    Sprite Size Mode=::NewEnumerator3
    Uniform Sprite Size=50
    UsePositionOffset=false
    Write Material Random=true
   )
   &CameraOffset(Camera Offset Amount=5)
  ]
  particleUpdateScript [
   &ParticleState(
    Kill Particles When Lifetime Has Elapsed=false
    Let Infinitely Lived Particles Die When Emitter Deactivates=true
    Loop Particles Lifetime=true
   )
   &SkeletalMeshLocation(
    Apply Sampled Orientation to Mesh Particles=true
    Inherit Velocity=false
    "Inherit Velocity (Scale)"=0
    Interpolate Mesh Sampling=false
    MaskSpawnGroup=false
    Mesh Sampling Type=::NewEnumerator5
    Sampled Position Offset Coordinate Space:ENiagaraMeshTransforms=::NewEnumerator1
    Skeletal Mesh=$User.SkeletalMesh
    Socket Sampling Mode=::NewEnumerator5
   ) {module=SkeletalMeshLocation001}
   &ScaleColor(
    Scale RGB=&Multiply_VectorByFloat(
     Vector=<1 1 1>
     Float=&FloatFromCurve(
      FloatCurve=DataInterface() {propertyValues {Curve {keys [<at 0 0.8>,<at 0.1 0.9>,<at 0.3 0.7>,<at 0.4 0.8>,<at 0.5 0.8>,<at 0.6 1>,<at 0.8 0.75>,<at 0.9 0.95>,<at 1 0.8>]}}}
      Scale Curve=&Multiply_Float(A=1, B=1)
     )
    )
    ScaleA=true
    ScaleRGB=true
   ) {enabled=false}
   /Niagara/Modules/Update/Color/ScaleColorBySpeed(Max Speed Threshold=500, Min Speed Threshold=0, RGB Scale Min=0.1, Source Velocity:Vector3f=$Engine.Owner.Velocity) {enabled=false}
   &ScaleColor(
    Scale Alpha=&FloatFromCurve(FloatCurve=DataInterface() {propertyValues {Curve {keys [<at 0 1>,<at 1 0>]}}})
    ScaleA=true
    ScaleRGB=true
   ) {module=ScaleColor001}
  ]
  rendererClasses [&NiagaraSpriteRendererProperties]
  renderers [{rendererClass=&NiagaraSpriteRendererProperties}]
 }
 Glow002 {
  emitterUpdateScript [
   &EmitterState()
   &SpawnBurst_Instantaneous(Use Spawn Probability=false, UseLoopCountLimit=false)
  ]
  particleSpawnScript [
   &InitializeParticle(
    Color=<rgba 10 0 6.4152 1>
    Lifetime=1.1
    Mesh Scale Mode=::NewEnumerator3
    Mesh Uniform Scale=1.2
    Sprite Size Mode=::NewEnumerator3
    Uniform Sprite Size=65
    UsePositionOffset=false
    Write Material Random=true
   )
   &CameraOffset()
   &InitialMeshOrientation(Mesh Orientation Mode=::NewEnumerator7, Rotation=<0 -0.25 0>, Use Rotation Vector=true)
   /SetVariables(Particles.New float=10) {isSetParameters=true}
  ]
  particleUpdateScript [
   &ParticleState()
   &ScaleMeshSize(Scale Factor=&VectorFromFloat(Value=&FloatFromCurve(
      FloatCurve=DataInterface() {propertyValues {Curve {keys [
          {interpMode=RCIM_Cubic, tangentMode=RCTM_User, time=0, value=1.3}
          {arriveTangent=-0.4286, interpMode=RCIM_Cubic, leaveTangent=-0.4286, time=0.2, value=1}
          {arriveTangent=-0.25, interpMode=RCIM_Cubic, leaveTangent=-0.25, time=0.7, value=1}
          <at 1 0.8>
         ]}}}
      Scale Curve=0.9
     )))
   &SkeletalMeshLocation(
    Apply Sampled Orientation to Mesh Particles=true
    Bone Sampling Mode=::NewEnumerator4
    Inherit Velocity=false
    "Inherit Velocity (Scale)"=0
    Interpolate Mesh Sampling=false
    MaskSpawnGroup=false
    Sampled Position Offset=<0 1 0>
    Sampled Position Offset Coordinate Space:ENiagaraMeshTransforms=::NewEnumerator2
    Skeletal Mesh=$User.SkeletalMesh
   ) {module=SkeletalMeshLocation001}
   &FN_Gen5_Light() {enabled=false}
   &FN_Light_Attributes(
    Color=<rgba 1 0.081 0.8928 1>
    Create Color Attribute=true
    Create Exponent Attribute=true
    Create Position Attribute=true
    Create Radius Attribute=true
    Exponent=4
    Position=&AddVectorToPosition(Position=Linked() {linkedVariable {name=Output.SkeletalMeshLocation001.SampledPosition, type=Vector3f}}, Vector=<0 -5 0>)
    Radius=40
    Scale Color=&FloatFromCurve(FloatCurve=DataInterface() {propertyValues {Curve {keys [<at 0 1>,<at 0.8 0>]}}}, Scale Curve=10)
   )
   &DynamicMaterialParameters(
    Index 0 Param 1=&FloatFromCurve(
     FloatCurve=DataInterface() {propertyValues {Curve {keys [<at 0 1>,<at 1 0>]}}}
     Scale Curve=&Multiply_Float(
      A=&FNTimeOfDayScalar(Night Scale=0.2, Morning Scale=0.75, Day Scale=0.75, Evening Scale=0.2)
      B=140
     )
    )
    Module.Param0WriteEnabled=true
    Module.Param1WriteEnabled=true
    Module.Param2WriteEnabled=true
    Module.Param3WriteEnabled=true
   )
   &ScaleColor(
    Scale Alpha=&FloatFromCurve(FloatCurve=DataInterface() {propertyValues {Curve {keys [<at 0 1>,<at 1 0>]}}}, Scale Curve=0.1)
    Scale RGB=&VectorFromFloat(Value=&FloatFromCurve(
      FloatCurve=DataInterface() {propertyValues {Curve {keys [
          {interpMode=RCIM_Cubic, tangentMode=RCTM_User, time=0, value=1}
          {interpMode=RCIM_Cubic, tangentMode=RCTM_User, time=0.2, value=2}
          {interpMode=RCIM_Cubic, tangentMode=RCTM_User, time=1, value=0}
         ]}}}
      Scale Curve=8
     ))
    ScaleA=true
    ScaleRGB=true
   )
  ]
  rendererClasses [&NiagaraMeshRendererProperties, &NiagaraLightRendererProperties]
  renderers [
   {rendererClass=&NiagaraMeshRendererProperties, rendererIndex=0}
   {rendererClass=&NiagaraLightRendererProperties, rendererIndex=1}
   {rendererClass=&NiagaraMeshRendererProperties, rendererIndex=2}
   {rendererClass=&NiagaraMeshRendererProperties, rendererIndex=3}
  ]
 }
 PixelArtShapes001 {
  emitterUpdateScript [
   &EmitterState(Life Cycle Mode=::NewEnumerator1, Loop Behavior=::NewEnumerator1, Loop Delay=0.08, Loop Duration=0.2, UseLoopDelay=true)
   &SpawnRate(
    SpawnRate=&FloatFromCurve(
     FloatCurve=DataInterface() {propertyValues {Curve {keys [<at 0.05 0.5>,<at 0.2 1>,<at 1 0>]}}}
     CurveIndex=$Emitter.NormalizedLoopAge
     Scale Curve=200
    )
    Use Spawn Probability=false
   ) {enabled=false}
   &SpawnBurst_Instantaneous(
    Spawn Count=&RandomRangeInt(Minimum=7, Maximum=5)
    Use Spawn Probability=false
    UseLoopCountLimit=false
   )
  ]
  particleSpawnScript [
   &InitializeParticle(
    Color=&SelectLinearColorFromArray(
     Array Sampling Mode=::NewEnumerator0
     Direct Array Index=?
     Color Selection Array=DataInterface() {propertyValues {ColorData [<rgba 3 1.1206 0.875 1>, <rgba 3 0.7969 2.6778 1>, <rgba 0.8125 3 1.2203 1>, <rgba 0.8281 1.8712 3 1>]}}
     Random Seed=0
    )
    Lifetime=&Multiply_Float(
     A=&FloatFromCurve(FloatCurve=DataInterface() {propertyValues {Curve {keys [<at 0 1>,<at 0.5 0.3>]}}}, CurveIndex=$Emitter.NormalizedLoopAge)
     B=&RandomRangeFloat(Minimum=0.2, Maximum=0.7)
    )
    Sprite Rotation Angle=&SelectFloatFromArray(
     Array Sampling Mode=::NewEnumerator0
     Direct Array Index=?
     Float Selection Array=DataInterface() {propertyValues {FloatData [0.15,0]}}
     Random Seed=0
    )
    Sprite Rotation Mode=::NewEnumerator1
    Sprite Size Mode=::NewEnumerator1
    Uniform Sprite Size Max=4
    Uniform Sprite Size Min=0.5
    UsePositionOffset=false
    Write Material Random=true
   )
   &InheritVelocity(Inherited Velocity Amount Scale=&VectorFromFloat(Value=0.5), Speed Limit=true, UseSpeedThreshold=false) {enabled=false}
   &AddVelocityFromPoint(Module.Use Falloff Distance=false, Velocity Strength=&RandomRangeFloat(Minimum=400, Maximum=500)) {enabled=false}
   &ApplyInitialForces()
   &CameraOffset(Camera Offset Amount=10)
   &SkeletalMeshLocation(
    MaskSpawnGroup=false
    Mesh Sampling Type=::NewEnumerator5
    Sampled Position Offset=<2 0 0>
    Sampled Position Offset Coordinate Space:ENiagaraSkelMeshTransforms=::NewEnumerator2
    Skeletal Mesh=$User.SkeletalMesh
    Socket Sampling Mode=::NewEnumerator5
   )
   &ShapeLocation(
    MaskSpawnGroup=false
    Ring / Disc Distribution Mode:ENiagaraSphereDistributionMode=::NewEnumerator3
    Ring Radius=&RandomRangeFloat(Minimum=10, Maximum=15)
    Rotation Axis=<0 1 0>
    Rotation Mode=::NewEnumerator1
    Shape Primitive=::NewEnumerator4
    Uniform Spiral Amount:NiagaraFloat=&RandomRangeFloat(Minimum=2, Maximum=2.4)
    Uniform Spiral Falloff=0
    UseOwnerScale=true
   ) {enabled=false}
   &ShapeLocation(
    Hemisphere Distribution=&MakeVector2D(
     X=0.5
     Y=&Lerp_Float(A=1, B=&Multiply_Float(A=0.5, B=&ReturnNormalizedExecIndex()), Alpha=0)
    )
    MaskSpawnGroup=false
    Non Uniform Scale=<1 -1 1>
    Override Seed=false
    Sphere Radius=15
    Sphere Surface Distribution=1
    UseOwnerScale=true
   ) {module=ShapeLocation003}
   &AddVelocity(Module.Use Falloff Distance=false, Override Seed=false, Velocity Mode=::NewEnumerator1, Velocity Speed=0)
  ]
  particleUpdateScript [
   &Drag(Drag=&RandomRangeFloat(Minimum=2, Maximum=5), Use Linear Drag=true, Use Rotational Drag=false)
   &SubUVAnimation(End Frame=4, Start Frame=1) {module=SubUVAnimation001}
   &GravityForce(Gravity=<0 0 100>)
   &SolveForcesAndVelocity(Clamp Velocity=false, Limit Acceleration=false)
   &SubUVAnimation(End Frame=4, Start Frame=2) {enabled=false}
   &ParticleState()
   &ScaleSpriteSize(Uniform Curve Sprite Scale=DataInterface() {propertyValues {Curve {keys [
        {arriveTangent=5.201, interpMode=RCIM_Constant, time=0, value=1}
        <at 0.5 0.5 RCIM_Constant>
        {arriveTangent=-1.983, interpMode=RCIM_Constant, leaveTangent=-1.983, time=1, value=0}
       ]}}})
   &ScaleSpriteSize(
    Uniform Curve Scale=2
    Uniform Curve Sprite Scale=DataInterface() {propertyValues {Curve {keys [<at 0 1.5 RCIM_Constant>,<at 0.25 1 RCIM_Constant>,<at 0.5 0.4 RCIM_Constant>,<at 0.75 0.25 RCIM_Constant>,<at 1 0 RCIM_Constant>]}}}
   ) {enabled=false, module=ScaleSpriteSize001}
   &DynamicMaterialParameters(
    Index 0 Param 1=0.02
    Module.Param0WriteEnabled=true
    Module.Param1WriteEnabled=true
    Module.Param2WriteEnabled=true
    Module.Param3WriteEnabled=true
   )
   &self:Quantizer_05(SnapDistance=4) {enabled=false, module=Quantizer_05}
   /SetVariables(Particles.Lifetime=&Multiply_Float(
     A=$Particles.Initial.Lifetime
     B=&FloatFromCurve(FloatCurve=DataInterface() {propertyValues {Curve {keys [<at 0 1>,<at 1 0.3>]}}}, CurveIndex=$Emitter.NormalizedLoopAge)
    )) {enabled=false, isSetParameters=true, module=SetVariables_#1}
   &ScaleColor(
    Scale Alpha=&FloatFromCurve(FloatCurve=DataInterface() {propertyValues {Curve {keys [<at 0 1>,<at 1 1>]}}})
    Scale RGB=&VectorFromFloat(Value=&FloatFromCurve(FloatCurve=DataInterface() {propertyValues {Curve {keys [<at 0 1>,<at 1 1>]}}}, Scale Curve=0.5))
    ScaleA=true
    ScaleRGB=true
   )
  ]
  rendererClasses [&NiagaraSpriteRendererProperties]
  renderers [
   {rendererClass=&NiagaraSpriteRendererProperties}
   {rendererClass=&NiagaraSpriteRendererProperties, rendererIndex=1}
  ]
 }
}
summary {
 systemName=NS_Veiled_Reactive_Burst
 emitters [
  {emitterName=Glow, rendererClasses [&NiagaraSpriteRendererProperties]}
  {bEnabled=false, emitterName=DigitalSquares, rendererClasses [&NiagaraSpriteRendererProperties]}
  {emitterName=DigitalSquares003, rendererClasses [&NiagaraSpriteRendererProperties]}
  {emitterName=Glow002, rendererClasses [&NiagaraMeshRendererProperties, &NiagaraLightRendererProperties]}
  {emitterName=Glitch_sprite, rendererClasses [&NiagaraMeshRendererProperties]}
  {bEnabled=false, emitterName=Glitch_sprite001, rendererClasses [&NiagaraMeshRendererProperties]}
  {emitterName=PixelArtShapes001, rendererClasses [&NiagaraSpriteRendererProperties]}
  {bEnabled=false, emitterName=Cubes, rendererClasses [&NiagaraSpriteRendererProperties]}
 ]
 userVariables [
  User.bIsFrontEnd {type=NiagaraBool}
  User.bIsFrontEndPreview {type=NiagaraBool}
  User.ColorArray {
   type=NiagaraDataInterfaceArrayColor
   description=""
   defaultValue {
    struct=NiagaraExt_VariableValue_DataInterface
    value {
     dataInterface=&self:NiagaraDataInterfaceArrayColor_0
     dataInterfaceClass=&NiagaraDataInterfaceArrayColor
    }
   }
  }
  User.EmissiveColorSelection {type=NiagaraFloat, defaultValue=0, description=""}
  User.SkeletalMesh {
   type=NiagaraDataInterfaceSkeletalMesh
   defaultValue {
    struct=NiagaraExt_VariableValue_DataInterface
    value {dataInterface=&self:NiagaraDataInterfaceSkeletalMesh_0}
   }
  }
 ]
}
